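# Niagara system for the ice elemental weapon skin, shedding drifting snowflakes and cold mist from a cylinder.
# Elemental_Weapon_Setup runs at system level, while both GPU sprite emitters read System cylinder dimensions and spawn-rate scale.
# Spawn groups vary size and set VisibilityTag across multiple renderers.
# Flakes also spawn by distance, inherit velocity and swirl in a rotating vector field; Mist sinks, fading through curve-driven alpha and material parameters.
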NiagaraSystem /Game/Weapons/WeaponSkins/Elemental/Ice/NS_Elemental_Ice
dependencies {
 usedDataInterfaces [
  &NiagaraDataInterfaceCurve
  &NiagaraDataInterfaceSpriteRendererInfo
  /Script/Niagara.NiagaraDataInterfaceVectorField
  &NiagaraDataInterfaceSkeletalMesh
 ]
 usedDynamicInputs []
 usedModules [&self:Elemental_Weapon_Setup, &SystemState]
}
emitters {
 Flakes {
  simTarget=GPUComputeSim
  emitterUpdateScript [
   &EmitterState()
   &SpawnRate(
    SpawnRate=&Multiply_Float(A=&RandomRangeFloat(Minimum=15, Maximum=60), B=$System.SpawnRateScale)
    Use Spawn Probability=false
   )
   &SpawnPerUnit(Module.Use Max Movement Threshold=true, Module.Use Movement Tolerance=false, Use Spawn Probability=false)
   &SpawnRate(
    SpawnGroup=1
    SpawnRate=&Multiply_Float(A=&RandomRangeFloat(Minimum=0.5, Maximum=1.5), B=$System.SpawnRateScale)
    Use Spawn Probability=false
   ) {module=SpawnRate001}
   &SpawnPerUnit(Module.Use Max Movement Threshold=true, Module.Use Movement Tolerance=false, Spawn Group=1, Spawn Spacing=150, Use Spawn Probability=false) {module=SpawnPerUnit001}
  ]
  particleSpawnScript [
   /SetVariables(Particles.VisibilityTag=&SpawnGroup()) {isSetParameters=true}
   &InitializeParticle(
    Color=<rgba 0.3769 0.6032 1 1>
    Lifetime=&MaskFloatBySpawnGroup(
     Masked Float=&RandomRangeFloat(Minimum=2, Maximum=3)
     Passthrough Float=&RandomRangeFloat(Minimum=1.5, Maximum=2)
    )
    Sprite Rotation Mode=::NewEnumerator2
    Sprite Size Mode=::NewEnumerator3
    Uniform Sprite Size=&Multiply_Float(
     A=&MaskFloatBySpawnGroup(
      Masked Float=&RandomRangeFloat(Minimum=0.6, Maximum=1.6)
      Passthrough Float=&RandomRangeFloat(Minimum=3, Maximum=5)
     )
     B=&RandomRangeFloat(Minimum=0.25, Maximum=0.5)
    )
    UsePositionOffset=false
    Write Material Random=true
   )
   &ShapeLocation(
    Cylinder Height=$System.CylinderHeight
    Cylinder Radius=$System.CylinderRadius
    MaskSpawnGroup=false
    Override Seed=false
    Rotation Mode=::NewEnumerator1
    Shape Origin=$System.CylinderStartLocation
    Shape Primitive=::NewEnumerator1
    Surface Only=true
    Use Endcaps In Surface Only Mode=false
    UseOwnerScale=true
   )
   &AddVelocity(Velocity=&Add_Vector(
     A=&Multiply_VectorByFloat(Vector=$Particles.ShapeLocation.ShapeNormal, Float=&RandomRangeFloat(Minimum=0, Maximum=0.02))
     B=&RandomRangeVector(Minimum=<0 5 0>, Maximum=<0 20 0>)
    ))
  ]
  particleUpdateScript [
   &ParticleState()
   &ScaleSpriteSize(Uniform Curve Sprite Scale=DataInterface() {propertyValues {Curve {keys [
        <at 0 0 RCIM_Cubic>
        {arriveTangentWeight=0.0972, interpMode=RCIM_Cubic, leaveTangentWeight=0.2915, tangentMode=RCTM_User, time=0.15, value=1}
        {arriveTangentWeight=0.2316, interpMode=RCIM_Cubic, leaveTangentWeight=0.2059, tangentMode=RCTM_User, time=0.6, value=1}
        <at 1 0 RCIM_Cubic>
       ]}}})
   &AccelerationForce(Acceleration=<0 0 -25>)
   &InheritVelocity(
    Inherited Velocity Amount Scale=&VectorFromFloat(Value=0.1)
    Inherited Velocity Speed Limit=12
    Speed Limit=true
    UseSpeedThreshold=false
   )
   &Drag(Drag=0.5, Use Linear Drag=true, Use Rotational Drag=false)
   /Niagara/VectorFields/SampleVectorField(
    FieldRotate:Quat4f=&MakeQuatFromAxisAngle(Axis=<0.335 -0.415 0.152>, AngleInDegrees=&Multiply_Float(A=360, B=$Engine.DeltaTime), NormalizeAxis=true)
    FieldScale:Vector3f=<3 3 3>
    Sampled Vector Scale=10
    VectorField=DataInterface() {propertyValues {Field=/Game/Effects/VectorFields/VF_TurbineSwirl_01, bTileX=true, bTileY=true, bTileZ=true}}
   )
   /Niagara/VectorFields/ApplyVectorField(Vector Field Velocity Tightness=0.05)
   &SolveForcesAndVelocity(Clamp Velocity=false, Limit Acceleration=false)
  ]
  rendererClasses [&NiagaraSpriteRendererProperties]
  renderers [
   {rendererClass=&NiagaraSpriteRendererProperties}
   {rendererClass=&NiagaraSpriteRendererProperties, rendererIndex=1}
  ]
 }
 Mist {
  simTarget=GPUComputeSim
  emitterUpdateScript [
   &EmitterState()
   &SpawnRate(SpawnRate=&Multiply_Float(A=3.5, B=$System.SpawnRateScale), Use Spawn Probability=false)
   &SpawnRate(SpawnGroup=1, SpawnRate=&Multiply_Float(A=1.5, B=$System.SpawnRateScale), Use Spawn Probability=false) {module=SpawnRate001}
   &SpawnRate(SpawnGroup=2, SpawnRate=&Multiply_Float(A=2.5, B=$System.SpawnRateScale), Use Spawn Probability=false) {module=SpawnRate002}
  ]
  particleSpawnScript [
   /SetVariables(Particles.VisibilityTag=&SpawnGroup()) {isSetParameters=true}
   &InitializeParticle(
    Color=&MakeLinearColorFromVectorAndFloat(
     "Vector (RGB)"=&MakeVectorFromLinearColorRGB(Color=<rgba 0.22 0.5722 1 1>)
     "Float (Alpha)"=&RandomRangeFloat(Minimum=0.04, Maximum=0.06)
    )
    Lifetime=&MaskFloatBySpawnGroup(
     Masked Float=&RandomRangeFloat(Minimum=0.9, Maximum=1.1)
     Passthrough Float=&RandomRangeFloat(Minimum=1.5, Maximum=2)
    )
    Sprite Size Mode=::NewEnumerator3
    Uniform Sprite Size=&MaskFloatBySpawnGroup(
     Masked Float=&RandomRangeFloat(Minimum=15, Maximum=25)
     Passthrough Float=&RandomRangeFloat(Minimum=35, Maximum=50)
    )
    UsePositionOffset=false
    Write Material Random=true
   )
   &ShapeLocation(
    Cylinder Height=$System.CylinderHeight
    Cylinder Radius=$System.CylinderRadius
    MaskSpawnGroup=false
    Override Seed=false
    Rotation Mode=::NewEnumerator1
    Shape Origin=$System.CylinderStartLocation
    Shape Primitive=::NewEnumerator1
    Surface Only=true
    Use Endcaps In Surface Only Mode=false
    UseOwnerScale=true
   )
   &AddVelocity(Velocity=&Add_Vector(
     A=&Multiply_VectorByFloat(Vector=$Particles.ShapeLocation.ShapeNormal, Float=&RandomRangeFloat(Minimum=0, Maximum=0.02))
     B=&MaskVectorBySpawnGroup(
      Masked Vector=&RandomRangeVector(Minimum=<0 5 0>, Maximum=<0 10 0>)
      Passthrough Vector=&RandomRangeVector(Minimum=<0 1 -2>, Maximum=<0 3 -3>)
     )
    ))
  ]
  particleUpdateScript [
   &ParticleState()
   &SpriteRotationRate(Rotation Rate=&RandomRangeFloat(Minimum=0, Maximum=0.01))
   &ScaleSpriteSize(
    Uniform Curve Scale=&MakeFloatFromVector(Vector=$System.SizeScale, Channel=::NewEnumerator0)
    Uniform Curve Sprite Scale=DataInterface() {propertyValues {Curve {keys [
        {arriveTangent=1.5681, interpMode=RCIM_Cubic, time=0, value=0.3}
        {arriveTangent=0.7, interpMode=RCIM_Cubic, leaveTangent=0.7, time=0.3, value=0.6}
        <at 1 1 RCIM_Cubic>
       ]}}}
   )
   &DynamicMaterialParameters(
    Index 0 Param 1=&FloatFromCurve(FloatCurve=DataInterface() {propertyValues {Curve {keys [
         {arriveTangent=4.2695, interpMode=RCIM_Cubic, leaveTangent=4.2695, tangentMode=RCTM_User, time=0, value=-0.5}
         <at 0.3 0.25 RCIM_Cubic>
         {arriveTangent=-1.2843, interpMode=RCIM_Cubic, leaveTangent=-1.2843, tangentMode=RCTM_User, time=1, value=-0.5}
        ]}}})
    Index 0 Param 2=1
    Index 0 Param 3=1
    Index 0 Param 4=1
    Module.Index0Param0Write=true
    Module.Index0Param1Write=true
    Module.Index0Param2Write=true
    Module.Index0Param3Write=true
   )
   &AccelerationForce(Acceleration=&MaskVectorBySpawnGroup(
     Masked Vector=&RandomRangeVector(Minimum=<0 0 -10>, Maximum=<0 0 -40>)
     Passthrough Vector=&RandomRangeVector(Minimum=<0 0 -5>, Maximum=<0 0 -10>)
    ))
   &SolveForcesAndVelocity(Clamp Velocity=false, Limit Acceleration=false)
   &ScaleColor(
    Scale Alpha=&FloatFromCurve(FloatCurve=DataInterface() {propertyValues {Curve {keys [
         {arriveTangent=9.0725, interpMode=RCIM_Cubic, time=0, value=0}
         <at 0.25 1 RCIM_Cubic>
         {interpMode=RCIM_Cubic, leaveTangent=-0.0405, time=1, value=0}
        ]}}})
    ScaleA=true
    ScaleRGB=true
   )
   &SubUVAnimation(
    "Sprite Renderer "=DataInterface() {propertyValues {SpriteRenderer=&self:Mist_0.NiagaraSpriteRendererProperties_0}}
    UseEndFrame=true
    UseStartFrame=true
   )
  ]
  rendererClasses [&NiagaraSpriteRendererProperties]
  renderers [
   {rendererClass=&NiagaraSpriteRendererProperties}
   {rendererClass=&NiagaraSpriteRendererProperties, rendererIndex=1}
   {rendererClass=&NiagaraSpriteRendererProperties, rendererIndex=2}
  ]
 }
}
summary {
 systemName=NS_Elemental_Ice
 emitters [
  {emitterName=Mist, simTarget=GPUComputeSim, rendererClasses [&NiagaraSpriteRendererProperties]}
  {emitterName=Flakes, simTarget=GPUComputeSim, rendererClasses [&NiagaraSpriteRendererProperties]}
 ]
 userVariables []
}
